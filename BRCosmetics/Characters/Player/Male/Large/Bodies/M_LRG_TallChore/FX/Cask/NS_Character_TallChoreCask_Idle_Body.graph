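# Niagara idle body effect for the TallChore Cask character cosmetic, layering glows, wisp flames and motes on the skeletal mesh.
# Persistent burst emitters form purple sprite glows along bones and at spine_04, plus socket-pinned flame meshes scaled per index.
# DetailWhisps continuously spawns growing, fading meshes at four partitioned bone positions with array-driven offsets and orientations.
# Minimal adds small SubUV sprites that follow the skeleton, easing outward along random sphere offsets.
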NiagaraSystem /BRCosmetics/Characters/Player/Male/Large/Bodies/M_LRG_TallChore/FX/Cask/NS_Character_TallChoreCask_Idle_Body
dependencies {
 usedDataInterfaces [
  &NiagaraDataInterfaceCurve
  &NiagaraDataInterfaceArrayFloat3
  &NiagaraDataInterfaceSkeletalMesh
  &NiagaraDataInterfaceColorCurve
  &NiagaraDataInterfaceArrayInt32
  &NiagaraDataInterfaceSpriteRendererInfo
 ]
 usedDynamicInputs []
 usedModules [
  &self:SystemSpawnScript.SystemScriptSource.SystemScriptGraph.NiagaraNodeAssignment_0.SetVariables_#1
  &self:SystemSpawnScript.SystemScriptSource.SystemScriptGraph.NiagaraNodeAssignment_1.SetVariables_#2
  &SystemState
 ]
}
emitters {
 BodyGlow {
  emitterUpdateScript [
   &EmitterState(Life Cycle Mode=::NewEnumerator1, Loop Behavior=::NewEnumerator1, Loop Duration Mode=::NewEnumerator1, UseLoopDelay=false)
   &SpawnBurst_Instantaneous(Spawn Count=13, Use Spawn Probability=false, UseLoopCountLimit=false)
  ]
  particleSpawnScript [
   &PartitionParticles(Number Of Partitions=6) {enabled=false}
   &FN_SetSpawnGroup(Number of Spawn Groups=5)
   &InitializeParticle(
    Color=<rgba 0.8179 0.1347 1 0.2>
    Position Offset=<0 15 0>
    Sprite Size Mode=::NewEnumerator3
    Uniform Sprite Size=35
    UsePositionOffset=false
    Write Material Random=true
   )
  ]
  particleUpdateScript [
   &ParticleState(Kill Particles When Lifetime Has Elapsed=false)
   &SkeletalMeshLocation(
    Bone / Socket Index=&ReturnExecIndex()
    Bone Sampling Mode=::NewEnumerator4
    MaskSpawnGroup=false
    Sampled Position Offset=<0 11 0>
    Skeletal Mesh=$System.Skeletal Mesh Reader
   ) {enabled=false}
   &FN_BoneFlow(
    Angle=1
    Height:NiagaraFloat=&FloatFromCurve(
     FloatCurve=DataInterface() {propertyValues {Curve {defaultValue=1, keys [{tangentMode=RCTM_User, time=0, value=0}, {tangentMode=RCTM_User, time=1, value=1}]}}}
     CurveIndex=&ReturnNormalizedExecIndex()
    )
    Radius=1
    Set Material Dynamic=false
    SkeletalMesh:NiagaraDataInterfaceSkeletalMesh=$System.Skeletal Mesh Reader
   )
  ]
  rendererClasses [&NiagaraSpriteRendererProperties]
  renderers [{rendererClass=&NiagaraSpriteRendererProperties}]
 }
 ChestGlow {
  emitterUpdateScript [
   &EmitterState(Life Cycle Mode=::NewEnumerator1, Loop Behavior=::NewEnumerator1, Loop Duration Mode=::NewEnumerator1, UseLoopDelay=false)
   &SpawnBurst_Instantaneous(Use Spawn Probability=false, UseLoopCountLimit=false)
  ]
  particleSpawnScript [
   &PartitionParticles(Number Of Partitions=6) {enabled=false}
   &InitializeParticle(
    Color=<rgba 0.8179 0.1347 1 0.2>
    Position Offset=<0 15 0>
    Sprite Size Mode=::NewEnumerator3
    Uniform Sprite Size=100
    UsePositionOffset=false
    Write Material Random=true
   )
  ]
  particleUpdateScript [
   &ParticleState(Kill Particles When Lifetime Has Elapsed=false)
   &SkeletalMeshLocation(
    Bone Sampling Mode=::NewEnumerator4
    MaskSpawnGroup=false
    Skeletal Mesh=DataInterface() {propertyValues {
      DefaultMesh=null
      ExcludeBoneName=root
      OptionalComponentClass=null
      PreviewMesh=/BRCosmetics/Characters/Player/Male/Large/Bodies/M_LRG_TallChore/Meshes/M_LRG_TallChore
      SkinningMode=SkinOnTheFly
      SoftSourceActor=null
      SourceMode=Default
      UvSetIndex=0
      WholeMeshLOD=-1
      bExcludeBone=false
      bReadDeformedGeometry=true
      bRequireCurrentFrameData=true
      ComponentTags []
      FilteredBones [spine_04]
      FilteredSockets []
      MeshUserParameter {parameter {name=None, typeDefHandle {registeredTypeIndex=1}, varData []}}
      SamplingRegions []
     }}
   )
  ]
  rendererClasses [&NiagaraSpriteRendererProperties]
  renderers [{rendererClass=&NiagaraSpriteRendererProperties}]
 }
 DetailWhisps {
  emitterUpdateScript [
   &EmitterState()
   &SpawnRate(SpawnRate=4.5, Use Spawn Probability=false)
  ]
  particleSpawnScript [
   &PartitionParticles(Number Of Partitions=4)
   &InitializeParticle(
    Color=<rgba 1 1 1 0.2>
    Lifetime Max=1.6
    Lifetime Min=1.1
    Lifetime Mode=::NewEnumerator1
    Mesh Scale=&SelectVectorFromArray(
     Direct Array Index=$Particles.PartitionParticles.Partition
     Vector Selection Array=DataInterface() {propertyValues {FloatData [<0.1 0.1 0.2>, <0.1 0.1 0.2>, <0.08 0.08 0.15>, <0.08 0.08 0.15>]}}
    )
    Mesh Scale Mode=::NewEnumerator4
    UsePositionOffset=false
    Write Material Random=true
   )
   &InitialMeshOrientation(
    Mesh Orientation Mode=::NewEnumerator2
    Orientation Vector=&SelectVectorFromArray(
     Direct Array Index=$Particles.PartitionParticles.Partition
     Vector Selection Array=DataInterface() {propertyValues {FloatData [<0 0 -1>, <0 0 1>, <0 0 1>, <0 0 -1>]}}
    )
    Rotation=&RandomRangeVector(Minimum=<0 0 -1>, Maximum=<0 0 1>)
    Use Rotation Vector=true
   )
  ]
  particleUpdateScript [
   &ParticleState()
   &ScaleMeshSize(Scale Factor=&VectorFromFloat(Value=&FloatFromCurve(FloatCurve=DataInterface() {propertyValues {Curve {
         defaultValue=0
         keys [
          {arriveTangent=2.8496, interpMode=RCIM_Cubic, leaveTangent=-0.0069, tangentMode=RCTM_Break, time=0, value=0}
          {interpMode=RCIM_Cubic, tangentMode=RCTM_User, time=1, value=1}
         ]
        }}})))
   /SetVariables(Particles.Color3:LinearColor=&ColorFromCurve(ColorCurve=DataInterface() {propertyValues {
       AlphaCurve {keys [<at 0.0113 0>,<at 0.2063 0.3 RCIM_Cubic>,<at 0.5804 0.1>,<at 1 0 RCIM_Cubic>]}
       BlueCurve {keys [<at 0.4524 0.6163 RCIM_Cubic>]}
       GreenCurve {keys [<at 0.4524 0.35 RCIM_Cubic>]}
       RedCurve {keys [<at 0.4524 0.5575 RCIM_Cubic>]}
      }})) {isSetParameters=true}
   /SetVariables(Particles.GlowScale:Vector3f=&Multiply_Vector(A=$Particles.Scale, B=<1.2 1.2 0.96>)) {isSetParameters=true}
   &ScaleColor(
    Scale Alpha=&FloatFromCurve(FloatCurve=DataInterface() {propertyValues {Curve {keys [
         {arriveTangent=5.6559, interpMode=RCIM_Cubic, leaveTangent=5.6559, tangentMode=RCTM_User, time=0.0052, value=0.0275}
         {arriveTangent=1.2538, interpMode=RCIM_Cubic, leaveTangent=-0.0133, tangentMode=RCTM_Break, time=0.2895, value=0.9373}
         <at 1 0.0235>
        ]}}})
    Scale RGB=<1.2 1.2 1.2>
    ScaleA=true
    ScaleRGB=true
   )
   &SkeletalMeshLocation(
    Bone / Socket Index=&SelectIntFromArray(Direct Array Index=$Particles.PartitionParticles.Partition, Int Selection Array=DataInterface() {propertyValues {IntData [0,1,2,3]}})
    Bone Sampling Mode=::NewEnumerator4
    MaskSpawnGroup=false
    Sampled Position Offset=&SelectVectorFromArray(
     Direct Array Index=$Particles.PartitionParticles.Partition
     Vector Selection Array=DataInterface() {propertyValues {FloatData [<-26 0 0>, <26 0 0>, <22 0 0>, <-22 0 0>]}}
    )
    Skeletal Mesh=$System.SKM2
   )
  ]
  rendererClasses [&NiagaraMeshRendererProperties]
  renderers [
   {rendererClass=&NiagaraMeshRendererProperties, rendererIndex=0}
   {rendererClass=&NiagaraMeshRendererProperties, rendererIndex=1}
  ]
 }
 Minimal {
  emitterUpdateScript [
   &EmitterState()
   &SpawnRate(Spawn Probability=0.5, SpawnRate=12, Use Spawn Probability=true)
  ]
  particleSpawnScript [
   &InitializeParticle(
    Color=<rgba 1.1 0.7996 1.0938 1>
    Lifetime Max=1.5
    Lifetime Min=0.9
    Lifetime Mode=::NewEnumerator1
    Sprite Rotation Mode=::NewEnumerator2
    Sprite Size Mode=::NewEnumerator1
    Uniform Sprite Size Max=8
    Uniform Sprite Size Min=2
    UsePositionOffset=false
    Write Material Random=true
   )
   &SubUVAnimation(
    "Sprite Renderer "=DataInterface() {propertyValues {SpriteRenderer=&self:Minimal.NiagaraSpriteRendererProperties_1}}
    SubUV Animation Mode:ENiagara_SubUVLookupModeV2=::NewEnumerator2
    UseEndFrame=false
    UseStartFrame=false
   )
   &ShapeLocation(
    Apply To Particle Position=::NewEnumerator1
    MaskSpawnGroup=false
    Override Seed=false
    Sphere Radius=&RandomRangeFloat(Minimum=15, Maximum=25)
    UseOwnerScale=true
   )
  ]
  particleUpdateScript [
   &ParticleState()
   &ScaleSpriteSize(Uniform Curve Sprite Scale=DataInterface() {propertyValues {Curve {
       defaultValue=0
       keys [
        {arriveTangent=2.0667, interpMode=RCIM_Cubic, leaveTangent=2.0667, tangentMode=RCTM_User, time=0.0058, value=0.2388}
        {arriveTangent=-0.1858, interpMode=RCIM_Cubic, leaveTangent=-0.1858, tangentMode=RCTM_User, time=0.7, value=1}
       ]
      }}})
   &SkeletalMeshLocation(
    Bone Sampling Mode=::NewEnumerator0
    MaskSpawnGroup=false
    OverrideRandomSeed=false
    Sampled Position Offset=&Lerp_Vector(
     A=<0 0 0>
     B=Linked() {linkedVariable {name=Particles.Initial.ShapeLocation.ShapePosition, type=NiagaraPosition}}
     Alpha=&FloatFromCurve(FloatCurve=DataInterface() {propertyValues {Curve {
         defaultValue=1
         keys [
          {arriveTangent=0.3418, interpMode=RCIM_Cubic, leaveTangent=0.3418, tangentMode=RCTM_User, time=0.0023, value=0.4849}
          {arriveTangent=0.9545, interpMode=RCIM_Cubic, leaveTangent=0.9545, tangentMode=RCTM_User, time=0.927, value=0.9092}
         ]
        }}})
    )
    Skeletal Mesh=$System.Skeletal Mesh Reader
   )
   &CurlNoiseForce(Bias Noise=false, Noise Strength=50, Pan Noise=false) {enabled=false}
   &SolveForcesAndVelocity(Clamp Velocity=false, Limit Acceleration=false)
   &ScaleColor(
    Scale Alpha=&FloatFromCurve(FloatCurve=DataInterface() {propertyValues {Curve {
        defaultValue=0
        keys [
         {arriveTangent=2.7177, interpMode=RCIM_Cubic, leaveTangent=2.7177, tangentMode=RCTM_User, time=0, value=0.0141}
         {arriveTangent=-0.2739, interpMode=RCIM_Cubic, leaveTangent=-4.1164, tangentMode=RCTM_Break, time=0.5065, value=1}
         {arriveTangent=-0.0249, interpMode=RCIM_Cubic, leaveTangent=-0.0249, tangentMode=RCTM_User, time=0.9953, value=0}
        ]
       }}})
    ScaleA=true
    ScaleRGB=true
   )
  ]
  rendererClasses [&NiagaraSpriteRendererProperties]
  renderers [
   {rendererClass=&NiagaraSpriteRendererProperties}
   {rendererClass=&NiagaraSpriteRendererProperties, rendererIndex=1}
  ]
 }
 WhispFlames {
  emitterUpdateScript [
   &EmitterState(Life Cycle Mode=::NewEnumerator1, Loop Behavior=::NewEnumerator1, Loop Duration Mode=::NewEnumerator1, UseLoopDelay=false)
   &SpawnBurst_Instantaneous(Spawn Count=9, Use Spawn Probability=false, UseLoopCountLimit=false)
  ]
  particleSpawnScript [&InitializeParticle(
    Color=<rgba 1 1 1 0.2>
    Lifetime Mode=::NewEnumerator1
    Mesh Scale=&Multiply_VectorByFloat(
     Vector=&SelectVectorFromArray(
      Direct Array Index=&ReturnExecIndex()
      Vector Selection Array=DataInterface() {propertyValues {FloatData [<1.4 1.4 0.55>, <1.4 1.4 0.55>, <1 1 0.6>, <1 1 0.6>, <0.9 0.9 0.6>, <0.9 0.9 0.6>, <1.5 1.5 0.6>, <1.25 1.25 0.8>, <1.25 1.25 0.8>]}}
     )
     Float=1
    )
    Mesh Scale Mode=::NewEnumerator4
    Sprite Size Mode=::NewEnumerator3
    Uniform Sprite Size=2
    UsePositionOffset=false
    Write Material Random=true
   )]
  particleUpdateScript [
   &ParticleState(Kill Particles When Lifetime Has Elapsed=false)
   &CameraOffset(Camera Offset Amount=12)
   /SetVariables(Particles.Color2=<rgba 0.0323 0.1233 0.0495 0.3>) {isSetParameters=true}
   /SetVariables(Particles.Color3:LinearColor=<rgba 0.6172 0.3734 0.5932 0.12>) {isSetParameters=true}
   /SetVariables(Particles.GlowScale:Vector3f=&Multiply_Vector(A=$Particles.Scale, B=<1.2 1.2 0.96>)) {isSetParameters=true}
   &SkeletalMeshLocation(
    Bone / Socket Index=&ReturnExecIndex()
    MaskSpawnGroup=false
    Mesh Sampling Type=::NewEnumerator5
    Skeletal Mesh=$System.Skeletal Mesh Reader
    Socket Sampling Mode=::NewEnumerator5
   )
  ]
  rendererClasses [&NiagaraSpriteRendererProperties, &NiagaraMeshRendererProperties]
  renderers [
   {rendererClass=&NiagaraSpriteRendererProperties}
   {rendererClass=&NiagaraMeshRendererProperties, rendererIndex=1}
   {rendererClass=&NiagaraMeshRendererProperties, rendererIndex=2}
   {rendererClass=&NiagaraMeshRendererProperties, rendererIndex=3}
  ]
 }
}
summary {
 systemName=NS_Character_TallChoreCask_Idle_Body
 emitters [
  {emitterName=BodyGlow, rendererClasses [&NiagaraSpriteRendererProperties]}
  {emitterName=WhispFlames, rendererClasses [&NiagaraSpriteRendererProperties, &NiagaraMeshRendererProperties]}
  {emitterName=ChestGlow, rendererClasses [&NiagaraSpriteRendererProperties]}
  {emitterName=DetailWhisps, rendererClasses [&NiagaraMeshRendererProperties]}
  {emitterName=Minimal, rendererClasses [&NiagaraSpriteRendererProperties]}
 ]
 userVariables []
}
